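# Animation Blueprint for the M_MED_Clash_V backpack cosmetic, posing the pack on an idle animation and shifting its legs during skydiving.
# The AnimGraph plays the idle, then offsets the pack bone and scales down packScale.
# Each update ORs the skydiving, dive-mode and floating flags into Skydive Falling.
# That bool blends mirrored additive roll on both master leg joints in and out linearly; Took Damage is unused.
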AnimBlueprint /BRCosmetics/Accessories/FORT_Backpacks/Backpack_M_MED_Clash_V/Meshes/M_MED_Clash_V_Pack_AnimBP
target_skeleton=&/M_MED_Clash_V_Pack_Skeleton
graphs [
 AnimGraph {nodes [
   n1 = "Transform (Modify) Bone - Bone: pack"(
    ComponentPose=Local To Component(LocalPose=Sequence Player 'Clash_V_Idle_BACKPACK'() {sequence=/BRCosmetics/Accessories/FORT_Backpacks/Backpack_M_MED_Clash_V/Animations/Clash_V_Idle_BACKPACK})
    Translation="0,1,0"
    Rotation="0,0,0"
   ) {
    rotationMode=BMM_Replace
    rotationSpace=BCS_BoneSpace
    scaleMode=BMM_Replace
    scaleSpace=BCS_BoneSpace
    translationMode=BMM_Replace
    translationSpace=BCS_BoneSpace
    alphaBoolBlend {blendOption=Linear}
    boneToModify {boneName=pack}
    translation {y=1}
   }
   n2 = "Transform (Modify) Bone - Bone: packScale"(ComponentPose=@n1, Alpha=1, Translation="0,0,0", Rotation="0,0,0", Scale="0.9,0.9,0.9") {
    rotationMode=BMM_Replace
    rotationSpace=BCS_BoneSpace
    scaleMode=BMM_Replace
    scaleSpace=BCS_BoneSpace
    translationMode=BMM_Replace
    translationSpace=BCS_BoneSpace
    alphaBoolBlend {blendOption=Linear}
    boneToModify {boneName=packScale}
    scale=<0.9 0.9 0.9>
   }
   n3 = Get Skydive Falling() -> (Skydive Falling:bool)
   n4 = "Transform (Modify) Bone - Bone: masterlegJnt_LFT"!(ComponentPose=@n2, bAlphaBoolEnabled:bool=@n3, Translation="0,0,0", Rotation="0,0,25", Scale="1,1,1") -> (Pose) {
    alphaInputType=Bool
    rotationMode=BMM_Additive
    rotationSpace=BCS_BoneSpace
    scaleMode=BMM_Replace
    scaleSpace=BCS_BoneSpace
    translationMode=BMM_Replace
    translationSpace=BCS_BoneSpace
    alphaBoolBlend {blendInTime=1, blendOption=Linear, blendOutTime=1}
    boneToModify {boneName=masterlegJnt_LFT}
    rotation {roll=25}
   }
   n5 = "Transform (Modify) Bone - Bone: masterlegJnt_RGT"!(ComponentPose=@n4, bAlphaBoolEnabled:bool=@n3, Translation="0,0,0", Rotation="0.410,0,-25", Scale="1,1,1") -> (Pose) {
    alphaInputType=Bool
    rotationMode=BMM_Additive
    rotationSpace=BCS_BoneSpace
    scaleMode=BMM_Replace
    scaleSpace=BCS_BoneSpace
    translationMode=BMM_Replace
    translationSpace=BCS_BoneSpace
    alphaBoolBlend {blendInTime=1, blendOption=Linear, blendOutTime=1}
    boneToModify {boneName=masterlegJnt_RGT}
    rotation {pitch=0.41, roll=-25}
   }
   Output Pose(Result=Component To Local(ComponentPose=@n5)) {layerGroup=None}
  ]}
 EventGraph {nodes [
   Event BlueprintUpdateAnimation()
   > Set Skydive Falling(Skydive Falling:bool=OR Boolean!(A=Get bIsSkyDiving(), B=Get bIsSkydiveDiveMode(), C:bool=Get bIsSkydiveFloating()) -> (ReturnValue)) -> (Output_Get:bool)
   TryGetPawnOwner()
  ]}
]
variables [
 Skydive Falling {type=boolean}
 Took Damage {type=boolean}
]
